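# Physics asset for the beard face accessory on the male medium Guan Yu player skin, setting up the beard for physical motion.
# Six box bodies on the dyn_beard bones form a single constraint chain hanging from the head, each joint sharing moderate swing limits and very tight twist.
# Unconstrained sphere and capsule bodies on spine_01 and spine_05 approximate the torso, apparently as collision volumes for the beard.
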PhysicsAsset /BRCosmetics/Characters/Player/Male/Medium/Bodies/M_MED_GuanYu_01/Export/M_MED_Commando_Guan_Yu_FaceAcc_01_Beard_Physics
bodies {
 dyn_beard_01 [{shapeType=Box, extentX=3.5, extentY=10, extentZ=5.5, center=<2.6414 0 0>}]
 dyn_beard_02 [{shapeType=Box, extentX=3.5, extentY=10, extentZ=7, center=<2.6414 0 0>}]
 dyn_beard_03 [{shapeType=Box, extentX=3.5, extentY=10, extentZ=8, center=<2.6414 0 0>}]
 dyn_beard_04 [{shapeType=Box, extentX=3.5, extentY=10, extentZ=7.5, center=<2.6414 0 0>}]
 dyn_beard_05 [{shapeType=Box, extentX=3.5, extentY=10, extentZ=5.5, center=<2.6414 0 0>}]
 dyn_beard_06 [{shapeType=Box, extentX=3.5, extentY=10, extentZ=5.5, center=<2.6414 0 0>}]
 head [{shapeType=Capsule, length=0, radius=3.5}]
 spine_01 [{shapeType=Sphere, radius=18.75, center=<-4.8946 5.626 0>}]
 spine_05 [{shapeType=Capsule, length=19.75, radius=15, center=<-16.1051 -4.4136 0>, rotation=<rot 85.4037 -85.4014 -89.9979>}]
}
constraints [
 {bone1Name=dyn_beard_02, bone2Name=dyn_beard_01, swing1LimitDegrees=45, swing2LimitDegrees=45, twistLimitDegrees=2.5}
 {bone1Name=dyn_beard_03, bone2Name=dyn_beard_02, swing1LimitDegrees=45, swing2LimitDegrees=45, twistLimitDegrees=2.5}
 {bone1Name=dyn_beard_04, bone2Name=dyn_beard_03, swing1LimitDegrees=45, swing2LimitDegrees=45, twistLimitDegrees=2.5}
 {bone1Name=dyn_beard_05, bone2Name=dyn_beard_04, swing1LimitDegrees=45, swing2LimitDegrees=45, twistLimitDegrees=2.5}
 {bone1Name=dyn_beard_06, bone2Name=dyn_beard_05, swing1LimitDegrees=45, swing2LimitDegrees=45, twistLimitDegrees=2.5}
 {bone1Name=dyn_beard_01, bone2Name=head, swing1LimitDegrees=45, swing2LimitDegrees=45, twistLimitDegrees=2.5}
]
