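# Face accessory AnimBlueprint for the Guan Yu Commando body that simulates the beard with physics.
# It copies the parent mesh's pose into a RigidBody node using a beard physics asset, head-bone simulation space and overridden gravity.
# The event graph holds logic to tune clothing anim-drive spring and damper stiffness around a backpack check.
# That logic is not connected to the initialize or update events.
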AnimBlueprint /BRCosmetics/Characters/Player/Male/Medium/Bodies/M_MED_GuanYu_01/Export/M_MED_Commando_Guan_Yu_FaceAcc_01_AnimBP
target_skeleton=&/M_MED_Commando_Guan_Yu_FaceAcc_01_Skeleton
graphs [
 AnimGraph {nodes [
   n1 = RigidBody(ComponentPose=Local To Component(LocalPose=Copy Pose From Mesh() {bUseAttachedParent=true})) {
    bOverrideWorldGravity=true
    overridePhysicsAsset=&/M_MED_Commando_Guan_Yu_FaceAcc_01_Beard_Physics
    simulationSpace=BaseBoneSpace
    alphaBoolBlend {blendOption=Linear}
    baseBoneRef {boneName=head}
    componentLinearAccScale=<0.5 0.5 0.25>
    componentLinearVelScale=<0.5 0.5 0.25>
    overrideWorldGravity {z=-980}
   }
   Output Pose(Result=Component To Local(ComponentPose=@n1)) {layerGroup=None}
  ]}
 EventGraph {nodes [
   n1 = Cast To ClothingSimulationInteractorNv!(
    execute
    Object=GetClothingSimulationInteractor!(self:Skeletal Mesh Component Object Reference=GetPartSkeletalMeshComponent(), ClothingAssetName:Name=None) -> (ReturnValue:Clothing Simulation Interactor Object Reference)
   ) -> (then, CastFailed, AsClothing Simulation Interactor Nv:Clothing Simulation Interactor Nv Object Reference) {is=title K2Node_DynamicCast}
   > Set GuanYu_clothing_sim_interactor(GuanYu_clothing_sim_interactor:Clothing Simulation Interactor Nv Object Reference=@n1) -> (Output_Get:Clothing Simulation Interactor Nv Object Reference)
   n2 = Cast To FortPlayerPawn(Object=TryGetPawnOwner())
   > Is Valid(InputObject=GetSkeletalMeshForPartType(self=@n2, PartType=Backpack))
   > n3 = Set isWearingBackbling(isWearingBackbling=true)
   > n4 = Branch(Condition=@n3)
   n5 = Get GuanYu_clothing_sim_interactor() -> (GuanYu_clothing_sim_interactor:Clothing Simulation Interactor Nv Object Reference)
   > SetAnimDriveSpringStiffness!(execute, self:Clothing Simulation Interactor Object Reference=@n5, InStiffness:float=20) -> (then)
   > SetAnimDriveDamperStiffness!(execute, self:Clothing Simulation Interactor Nv Object Reference=@n5, InStiffness:float=10) -> (then)
   SetAnimDriveSpringStiffness(execute=@n4.else, self=@n5, InStiffness=0)
   > SetAnimDriveDamperStiffness(self=@n5, InStiffness=0)
   TryGetPawnOwner()
   Event BlueprintInitializeAnimation()
   Event BlueprintUpdateAnimation()
  ]}
]
variables [
 GuanYu_clothing_sim_interactor {type=ClothingSimulationInteractorNv, default=null}
 isWearingBackbling {type=boolean}
]
