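# Physics asset for the Guan Yu face accessory 01 on a medium male player character cosmetic, defining simple collision volumes for the upper body.
# Capsules cover the upper spine, neck and both clavicles, while the head combines a sphere and a capsule.
# It has no constraints, so the bodies are not jointed as a ragdoll and only supply collision shapes on their bones.
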PhysicsAsset /BRCosmetics/Characters/Player/Male/Medium/Bodies/M_MED_GuanYu_01/Export/M_MED_Commando_Guan_Yu_FaceAcc_01_Physics
bodies {
 clavicle_l [{shapeType=Capsule, length=7.3084, radius=7.9017, center=<7.4007 -0.9146 -0.2197>, rotation=<rot 90 0 0>}]
 clavicle_r [{shapeType=Capsule, length=7.3084, radius=7.9017, center=<-9.2405 -0.9146 0.4649>, rotation=<rot 90 0 0>}]
 head [
  {shapeType=Sphere, radius=6.6677, center=<1.3281 -0.8773 0>}
  {shapeType=Capsule, length=5.9657, radius=5.522, center=<0.8117 2.2232 0>, rotation=<rot 90 0 0>}
 ]
 neck_01 [{shapeType=Capsule, length=8.2568, radius=6.6364, rotation=<rot 90 0 10>}]
 spine_05 [{shapeType=Capsule, length=7.3601, radius=12.0816, center=<0.9682 0 -0.0287>}]
}
constraints []
