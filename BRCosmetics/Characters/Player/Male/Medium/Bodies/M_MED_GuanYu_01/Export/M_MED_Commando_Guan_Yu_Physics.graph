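# Physics asset for the Guan Yu male medium body cosmetic, covering pelvis, legs and dynamic chains hanging from both thighs, likely costume panels.
# The left thigh carries three three-link capsule chains; the right carries five four-link chains mixing boxes and capsules.
# Chain joints have tight twist and varied swing limits.
# Free-rotation joints tie the right front and back chain tips to the middle tip and link some leg bodies.
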PhysicsAsset /BRCosmetics/Characters/Player/Male/Medium/Bodies/M_MED_GuanYu_01/Export/M_MED_Commando_Guan_Yu_Physics
bodies {
 calf_l [{shapeType=Capsule, length=22, radius=12.4742, center=<-24.47 0.4044 0.1052>, rotation=<rot -85.084 85.4041 88.8707>}]
 calf_r [{shapeType=Capsule, length=26.25, radius=12.3492, center=<24.4993 -0.455 -0.1093>, rotation=<rot -85.084 85.4039 -89.8903>}]
 dyn_chain_l_back_01 [{shapeType=Capsule, length=15.25, radius=8.375, shapeName=dyn_chain_l_back_02_capsule, center=<8 5 7>, rotation=<rot -0.0001 0.0001 101.7858>}]
 dyn_chain_l_back_02 [{shapeType=Capsule, length=13, radius=11.625, shapeName=dyn_chain_l_back_02_capsule, center=<8 5 5>, rotation=<rot -0.0001 0.0001 101.7858>}]
 dyn_chain_l_back_03 [{shapeType=Capsule, length=13.375, radius=11.5, shapeName=dyn_chain_l_back_03_capsule, center=<5 5 5>, rotation=<rot -0.0001 0.0001 101.7858>}]
 dyn_chain_l_front_01 [{shapeType=Capsule, length=15.25, radius=8.375, shapeName=dyn_chain_l_front_02_capsule, center=<8 -5 7>, rotation=<rot 0.0001 0.0001 82.191>}]
 dyn_chain_l_front_02 [{shapeType=Capsule, length=13, radius=11.625, shapeName=dyn_chain_l_front_02_capsule, center=<8 -5 5>, rotation=<rot 0.0001 0.0001 82.191>}]
 dyn_chain_l_front_03 [{shapeType=Capsule, length=13.375, radius=11.5, shapeName=dyn_chain_l_front_03_capsule, center=<5 -5 5>, rotation=<rot 0.0001 0.0001 82.191>}]
 dyn_chain_l_mid_01 [{shapeType=Capsule, length=15.25, radius=8.375, shapeName=dyn_chain_l_mid_02_capsule, center=<8 0 7>, rotation=<rot 0 0 90>}]
 dyn_chain_l_mid_02 [{shapeType=Capsule, length=13, radius=11.625, shapeName=dyn_chain_l_mid_02_capsule, center=<8 0 5>, rotation=<rot 0 0 90>}]
 dyn_chain_l_mid_03 [{shapeType=Capsule, length=13.375, radius=11.5, shapeName=dyn_chain_l_mid_03_capsule, center=<5 0 5>, rotation=<rot 0 0 90>}]
 dyn_chain_r_b_01 [{shapeType=Box, extentX=14, extentY=10, extentZ=5, center=<8.7699 -1.2278 -0.1558>}]
 dyn_chain_r_b_02 [{shapeType=Capsule, length=14.875, radius=9.125, shapeName=dyn_chain_r_b_02_capsule, center=<5 -10 5>, rotation=<rot -0.0002 -0.0001 75.0001>}]
 dyn_chain_r_b_03 [{shapeType=Capsule, length=26.75, radius=11.25, shapeName=dyn_chain_r_b_03_capsule, center=<5 -10 5>, rotation=<rot -0.0002 -0.0001 75.0001>}]
 dyn_chain_r_b_04 [{shapeType=Capsule, length=26, radius=12, shapeName=dyn_chain_r_b_04_capsule, center=<5 -10 5>, rotation=<rot -0.0002 -0.0001 75.0001>}]
 dyn_chain_r_f_01 [{shapeType=Box, extentX=11.5, extentY=8.5, extentZ=5, center=<8.7213 1.2096 0.0003>}]
 dyn_chain_r_f_02 [{shapeType=Capsule, length=15.125, radius=8.875, shapeName=dyn_chain_r_f_02_capsule, center=<5 10 5>, rotation=<rot 0.0005 -0.0002 95.0002>}]
 dyn_chain_r_f_03 [{shapeType=Capsule, length=26.625, radius=11.375, shapeName=dyn_chain_r_f_03_capsule, center=<5 10 5>, rotation=<rot 0.0005 -0.0002 95.0002>}]
 dyn_chain_r_f_04 [{shapeType=Capsule, length=26, radius=12, shapeName=dyn_chain_r_f_04_capsule, center=<5 10 5>, rotation=<rot 0.0005 -0.0002 95.0002>}]
 dyn_chain_r_mid_01 [{shapeType=Box, extentX=14, extentY=10, extentZ=5, center=<8.8568 0 0>}]
 dyn_chain_r_mid_02 [{shapeType=Capsule, length=14.5, radius=9.5, shapeName=dyn_chain_r_mid_02_capsule, center=<5 0 5>, rotation=<rot 0 0 90>}]
 dyn_chain_r_mid_03 [{shapeType=Capsule, length=26, radius=12, shapeName=dyn_chain_r_mid_03_capsule, center=<5 0 5>, rotation=<rot 0 0 90>}]
 dyn_chain_r_mid_04 [{shapeType=Capsule, length=26, radius=12, shapeName=dyn_chain_r_mid_04_capsule, center=<5 0 5>, rotation=<rot 0 0 90>}]
 dyn_chain_r_mid_b_01 [{shapeType=Box, extentX=14, extentY=10, extentZ=5, center=<8.8207 -0.6992 -0.3856>}]
 dyn_chain_r_mid_b_02 [{shapeType=Box, extentX=14, extentY=20, extentZ=10.5, center=<8.8207 -0.6992 -0.3856>}]
 dyn_chain_r_mid_b_03 [{shapeType=Box, extentX=10, extentY=18, extentZ=10.5, center=<6.9695 -0.8053 -0.2421>}]
 dyn_chain_r_mid_b_04 [{shapeType=Box, extentX=9, extentY=20.5, extentZ=10.5, center=<6.1015 -0.8551 -0.1751>}]
 dyn_chain_r_mid_f_01 [{shapeType=Box, extentX=14, extentY=10, extentZ=5, center=<8.8029 0.8291 -0.5142>}]
 dyn_chain_r_mid_f_02 [{shapeType=Box, extentX=14, extentY=20, extentZ=5, center=<8.8029 0.8291 -0.5142>}]
 dyn_chain_r_mid_f_03 [{shapeType=Box, extentX=10, extentY=19, extentZ=10.5, center=<6.9811 0.7215 -0.1557>}]
 dyn_chain_r_mid_f_04 [{shapeType=Box, extentX=9, extentY=20.5, extentZ=10.5, center=<6.127 0.6709 0.0124>}]
 foot_l [{
   shapeType=Capsule
   length=15.6026
   radius=8.3392
   shapeName=foot_r_capsule
   center=<-3.4408 6.0916 -0.1404>
   rotation=<rot -1.6995 179.3705 96.3577>
  }]
 foot_r [{
   shapeType=Capsule
   length=15.6026
   radius=8.3392
   shapeName=foot_r_capsule
   center=<3.4408 -6.0916 0.1404>
   rotation=<rot -1.6995 179.3705 -83.6423>
  }]
 pelvis [{shapeType=Capsule, length=0, radius=10.75}]
 thigh_l [{shapeType=Capsule, length=37.75, radius=9.6397, center=<-21.3817 -0.8714 0.4546>, rotation=<rot -90 -179.9406 -0.0593>}]
 thigh_r [{shapeType=Capsule, length=31.75, radius=9.6397, center=<24.5715 0.8714 -0.564>, rotation=<rot -87.4097 -0.0005 0.0004>}]
}
constraints [
 {bone1Name=dyn_chain_l_mid_01, bone2Name=thigh_l, swing1LimitDegrees=35, swing2LimitDegrees=5, twistLimitDegrees=5}
 {bone1Name=dyn_chain_l_mid_02, bone2Name=dyn_chain_l_mid_01, swing1LimitDegrees=5, swing2LimitDegrees=35, twistLimitDegrees=5}
 {bone1Name=dyn_chain_l_mid_03, bone2Name=dyn_chain_l_mid_02, swing1LimitDegrees=5, swing2LimitDegrees=35, twistLimitDegrees=5}
 {bone1Name=dyn_chain_l_front_01, bone2Name=thigh_l, swing1LimitDegrees=35, swing2LimitDegrees=5, twistLimitDegrees=5}
 {bone1Name=dyn_chain_l_front_02, bone2Name=dyn_chain_l_front_01, swing1LimitDegrees=5, swing2LimitDegrees=35, twistLimitDegrees=5}
 {bone1Name=dyn_chain_l_front_03, bone2Name=dyn_chain_l_front_02, swing1LimitDegrees=5, swing2LimitDegrees=35, twistLimitDegrees=5}
 {bone1Name=dyn_chain_l_back_01, bone2Name=thigh_l, swing1LimitDegrees=35, swing2LimitDegrees=5, twistLimitDegrees=5}
 {bone1Name=dyn_chain_l_back_02, bone2Name=dyn_chain_l_back_01, swing1LimitDegrees=5, swing2LimitDegrees=35, twistLimitDegrees=5}
 {bone1Name=dyn_chain_l_back_03, bone2Name=dyn_chain_l_back_02, swing1LimitDegrees=5, swing2LimitDegrees=35, twistLimitDegrees=5}
 {bone1Name=dyn_chain_r_mid_01, bone2Name=thigh_r, swing1LimitDegrees=15, swing2LimitDegrees=15, twistLimitDegrees=5}
 {bone1Name=dyn_chain_r_mid_02, bone2Name=dyn_chain_r_mid_01, swing1LimitDegrees=30, swing2LimitDegrees=30, twistLimitDegrees=5}
 {bone1Name=dyn_chain_r_mid_03, bone2Name=dyn_chain_r_mid_02, swing1LimitDegrees=30, swing2LimitDegrees=30, twistLimitDegrees=5}
 {bone1Name=dyn_chain_r_mid_04, bone2Name=dyn_chain_r_mid_03, swing1LimitDegrees=30, swing2LimitDegrees=30, twistLimitDegrees=5}
 {bone1Name=dyn_chain_r_mid_f_01, bone2Name=thigh_r, swing1LimitDegrees=15, swing2LimitDegrees=15, twistLimitDegrees=5}
 {bone1Name=dyn_chain_r_mid_f_02, bone2Name=dyn_chain_r_mid_f_01, swing1LimitDegrees=30, swing2LimitDegrees=30, twistLimitDegrees=5}
 {bone1Name=dyn_chain_r_mid_f_03, bone2Name=dyn_chain_r_mid_f_02, swing1LimitDegrees=45, swing2LimitDegrees=45, twistLimitDegrees=5}
 {bone1Name=dyn_chain_r_mid_f_04, bone2Name=dyn_chain_r_mid_f_03, swing1LimitDegrees=60, swing2LimitDegrees=60, twistLimitDegrees=5}
 {bone1Name=dyn_chain_r_mid_b_01, bone2Name=thigh_r, swing1LimitDegrees=30, swing2LimitDegrees=30, twistLimitDegrees=5}
 {bone1Name=dyn_chain_r_mid_b_02, bone2Name=dyn_chain_r_mid_b_01, swing1LimitDegrees=45, swing2LimitDegrees=45, twistLimitDegrees=5}
 {bone1Name=dyn_chain_r_mid_b_03, bone2Name=dyn_chain_r_mid_b_02, swing1LimitDegrees=45, swing2LimitDegrees=45, twistLimitDegrees=5}
 {bone1Name=dyn_chain_r_mid_b_04, bone2Name=dyn_chain_r_mid_b_03, swing1LimitDegrees=60, swing2LimitDegrees=60, twistLimitDegrees=5}
 {bone1Name=dyn_chain_r_f_01, bone2Name=thigh_r, swing1LimitDegrees=15, swing2LimitDegrees=15, twistLimitDegrees=5}
 {bone1Name=dyn_chain_r_f_02, bone2Name=dyn_chain_r_f_01, swing1LimitDegrees=30, swing2LimitDegrees=30, twistLimitDegrees=5}
 {bone1Name=dyn_chain_r_f_03, bone2Name=dyn_chain_r_f_02, swing1LimitDegrees=30, swing2LimitDegrees=30, twistLimitDegrees=5}
 {bone1Name=dyn_chain_r_f_04, bone2Name=dyn_chain_r_f_03, swing1LimitDegrees=30, swing2LimitDegrees=30, twistLimitDegrees=5}
 {bone1Name=dyn_chain_r_b_01, bone2Name=thigh_r, swing1LimitDegrees=30, swing2LimitDegrees=30, twistLimitDegrees=5}
 {bone1Name=dyn_chain_r_b_02, bone2Name=dyn_chain_r_b_01, swing1LimitDegrees=30, swing2LimitDegrees=30, twistLimitDegrees=5}
 {bone1Name=dyn_chain_r_b_03, bone2Name=dyn_chain_r_b_02, swing1LimitDegrees=30, swing2LimitDegrees=30, twistLimitDegrees=5}
 {bone1Name=dyn_chain_r_b_04, bone2Name=dyn_chain_r_b_03, swing1LimitDegrees=30, swing2LimitDegrees=30, twistLimitDegrees=5}
 {
  bone1Name=dyn_chain_r_b_04
  bone2Name=dyn_chain_r_mid_04
  swing1LimitDegrees=45
  swing1Motion=Free
  swing2LimitDegrees=45
  swing2Motion=Free
  twistLimitDegrees=45
  twistMotion=Free
 }
 {
  bone1Name=dyn_chain_r_f_04
  bone2Name=dyn_chain_r_mid_04
  swing1LimitDegrees=45
  swing1Motion=Free
  swing2LimitDegrees=45
  swing2Motion=Free
  twistLimitDegrees=45
  twistMotion=Free
 }
 {
  bone1Name=foot_r
  bone2Name=calf_r
  swing1LimitDegrees=45
  swing1Motion=Free
  swing2LimitDegrees=45
  swing2Motion=Free
  twistLimitDegrees=45
  twistMotion=Free
 }
 {
  bone1Name=thigh_l
  bone2Name=pelvis
  swing1LimitDegrees=45
  swing1Motion=Free
  swing2LimitDegrees=45
  swing2Motion=Free
  twistLimitDegrees=45
  twistMotion=Free
 }
 {
  bone1Name=calf_l
  bone2Name=thigh_l
  swing1LimitDegrees=45
  swing1Motion=Free
  swing2LimitDegrees=45
  swing2Motion=Free
  twistLimitDegrees=45
  twistMotion=Free
 }
 {
  bone1Name=foot_l
  bone2Name=calf_l
  swing1LimitDegrees=45
  swing1Motion=Free
  swing2LimitDegrees=45
  swing2Motion=Free
  twistLimitDegrees=45
  twistMotion=Free
 }
]
