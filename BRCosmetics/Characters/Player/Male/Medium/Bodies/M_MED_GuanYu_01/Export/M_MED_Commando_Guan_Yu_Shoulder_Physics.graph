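# Physics asset for the shoulder piece of the Guan Yu male medium Commando cosmetic body, likely for secondary motion of a dangling armor part.
# A tilted box body on the dyn_noseRing bone, apparently reused for the shoulder piece, pairs with a capsule on the left upper arm.
# A single constraint joins them with narrow twist and swing1 limits and a wide swing2, giving hinge-like movement relative to the arm.
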PhysicsAsset /BRCosmetics/Characters/Player/Male/Medium/Bodies/M_MED_GuanYu_01/Export/M_MED_Commando_Guan_Yu_Shoulder_Physics
bodies {
 dyn_noseRing [{shapeType=Box, extentX=5.6642, extentY=8.0249, extentZ=2.3978, center=<3.6936 0.0001 -1.7223>, rotation=<rot -24.9999 0 0>}]
 upperarm_l [{shapeType=Capsule, length=15, radius=4.25, center=<12.6783 0.0001 0>, rotation=<rot -90 -1.2659 1.2658>}]
}
constraints [{bone1Name=dyn_noseRing, bone2Name=upperarm_l, swing1LimitDegrees=2.5, swing2LimitDegrees=90, twistLimitDegrees=2.5}]
