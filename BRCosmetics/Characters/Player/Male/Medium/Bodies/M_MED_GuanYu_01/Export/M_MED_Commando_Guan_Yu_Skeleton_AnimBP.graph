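# Post-process style cosmetic AnimBP for the Guan Yu commando body, adding secondary motion to its armor, chains, sleeves and nose ring.
# It copies the parent mesh pose, then applies waist constraints, chain offsets, RigidBody and AnimDynamics into a cached pose.
# The event graph turns thigh, calf and arm rotations into corrective alphas.
# Skydiving or targeting selects Pose Driver shoulder pads, otherwise the constraint and arm-offset branch.
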AnimBlueprint /BRCosmetics/Characters/Player/Male/Medium/Bodies/M_MED_GuanYu_01/Export/M_MED_Commando_Guan_Yu_Skeleton_AnimBP
target_skeleton=&/M_MED_Commando_Guan_Yu_Skeleton
graphs [
 AnimGraph {nodes [
   RigidBody() {
    bOverrideWorldGravity=true
    overridePhysicsAsset=&/M_MED_Commando_Guan_Yu_Shoulder_Physics
    alphaBoolBlend {blendOption=Linear}
    componentAppliedLinearAccClamp=<250 250 250>
    componentLinearAccScale=<1 1 1>
    componentLinearVelScale=<1 1 1>
    overrideWorldGravity {z=-980}
   }
   n1 = "Constraint - pd_shoulderpad_l (PPPPPPPP)"!(
    ComponentPose=Local To Component(LocalPose=Use cached pose 'prePDCache'())
    Alpha=0.5
    ConstraintWeights_0=0.2273
    ConstraintWeights_1:float=0.2255
    ConstraintWeights_2:float=0.0016
    ConstraintWeights_3:float=0.0081
    ConstraintWeights_4:float=0.0027
    ConstraintWeights_5:float=0.0001
    ConstraintWeights_6:float=0.3557
    ConstraintWeights_7:float=0.179
   ) -> (Pose) {
    alphaBoolBlend {blendOption=Linear}
    boneToModify {boneName=pd_shoulderpad_l}
    constraintSetup [
     {targetBone {boneName=clavicle_l}}
     {targetBone {boneName=upperarm_twist_02_l}}
     {targetBone {boneName=spine_02}}
     {targetBone {boneName=spine_03}}
     {targetBone {boneName=spine_04}}
     {targetBone {boneName=spine_05}}
     {targetBone {boneName=upperarm_twist_01_l}}
     {targetBone {boneName=upperarm_l}}
    ]
    constraintWeights [1,1,1,1,1,1,1,1]
   }
   n2 = "Constraint - pd_shoulderpad_r (PPPPPPPPPPP)"!(
    ComponentPose=@n1
    Alpha=0.5
    ConstraintWeights_0=0.1447
    ConstraintWeights_1:float=0
    ConstraintWeights_2:float=0.4016
    ConstraintWeights_3:float=0.2529
    ConstraintWeights_4:float=0
    ConstraintWeights_5:float=0.0025
    ConstraintWeights_6:float=0.014
    ConstraintWeights_7:float=0.002
    ConstraintWeights_8:float=0.0029
    ConstraintWeights_9:float=0.1794
    ConstraintWeights_10:float=0
   ) -> (Pose) {
    alphaBoolBlend {blendOption=Linear}
    boneToModify {boneName=pd_shoulderpad_r}
    constraintSetup [
     {targetBone {boneName=upperarm_r}}
     {targetBone {boneName=clavicle_l}}
     {targetBone {boneName=upperarm_twist_01_r}}
     {targetBone {boneName=upperarm_twist_02_r}}
     {targetBone {boneName=spine_01}}
     {targetBone {boneName=spine_02}}
     {targetBone {boneName=spine_03}}
     {targetBone {boneName=spine_04}}
     {targetBone {boneName=spine_05}}
     {targetBone {boneName=clavicle_r}}
     {targetBone {boneName=pelvis}}
    ]
    constraintWeights [1,1,1,1,1,1,1,1,1,1,1]
   }
   n3 = "Transform (Modify) Bone - Bone: dyn_shoulderpad_l"(
    ComponentPose=@n2
    Alpha=Get arm_rot_up_l() -> (arm_rot_up_l:double)
    Translation="2.335,0.030,8.848"
    Rotation="16.084,-3.531,-0.981"
    Scale="1,1,1"
   ) {
    rotationMode=BMM_Additive
    rotationSpace=BCS_BoneSpace
    translationMode=BMM_Additive
    translationSpace=BCS_BoneSpace
    alphaBoolBlend {blendOption=Linear}
    boneToModify {boneName=dyn_shoulderpad_l}
    rotation=<rot 16.084 -3.531 -0.981>
    translation=<2.335 0.03 8.848>
   }
   n4 = Pose Driver - Source: upperarm_l!(SourcePose:Pose Link Structure=Use cached pose 'prePDCache'()) -> (Pose:Pose Link Structure) {
    is=title AnimGraphNode_PoseDriver
    driveSource=Translation
    poseAsset=&/M_MED_Commando_Guan_Yu_PoseDriver_PoseAsset
    onlyDriveBones [{boneName=pd_shoulderpad_l}]
    poseTargets [
     {
      bApplyCustomCurve=false
      bIsHidden=false
      distanceMethod=DefaultMethod
      drivenName=Pose_0
      functionType=DefaultFunction
      targetScale=1
      boneTransforms [{targetRotation=<rot -64.7854 -3.089 6.0959>, targetTranslation=<15.5111 0 0>}]
      customCurve {defaultValue=3.4028234663852886e+38, postInfinityExtrap=RCCE_Constant, preInfinityExtrap=RCCE_Constant, keys []}
      targetRotation=<rot 0 0 0>
     }
     {
      bApplyCustomCurve=false
      bIsHidden=false
      distanceMethod=DefaultMethod
      drivenName=Pose_3
      functionType=DefaultFunction
      targetScale=1
      boneTransforms [{targetRotation=<rot -80.0317 -12.5016 15.1498>, targetTranslation=<15.5111 0 0>}]
      customCurve {defaultValue=3.4028234663852886e+38, postInfinityExtrap=RCCE_Constant, preInfinityExtrap=RCCE_Constant, keys []}
      targetRotation=<rot 0 0 0>
     }
     {
      bApplyCustomCurve=false
      bIsHidden=false
      distanceMethod=DefaultMethod
      drivenName=Pose_4
      functionType=DefaultFunction
      targetScale=1
      boneTransforms [{targetRotation=<rot -64.7854 -3.089 71.4305>, targetTranslation=<15.5111 0 0>}]
      customCurve {defaultValue=3.4028234663852886e+38, postInfinityExtrap=RCCE_Constant, preInfinityExtrap=RCCE_Constant, keys []}
      targetRotation=<rot 0 0 0>
     }
     {
      bApplyCustomCurve=false
      bIsHidden=false
      distanceMethod=DefaultMethod
      drivenName=Pose_5
      functionType=DefaultFunction
      targetScale=1
      boneTransforms [{targetRotation=<rot -64.7854 -3.089 -89.7952>, targetTranslation=<15.5111 0 0>}]
      customCurve {defaultValue=3.4028234663852886e+38, postInfinityExtrap=RCCE_Constant, preInfinityExtrap=RCCE_Constant, keys []}
      targetRotation=<rot 0 0 0>
     }
     {
      bApplyCustomCurve=false
      bIsHidden=false
      distanceMethod=DefaultMethod
      drivenName=Pose_6
      functionType=DefaultFunction
      targetScale=1
      boneTransforms [{targetRotation=<rot 5.3225 96.1402 -64.8218>, targetTranslation=<15.5111 0 0>}]
      customCurve {defaultValue=3.4028234663852886e+38, postInfinityExtrap=RCCE_Constant, preInfinityExtrap=RCCE_Constant, keys []}
      targetRotation=<rot 0 0 0>
     }
     {
      bApplyCustomCurve=false
      bIsHidden=false
      distanceMethod=DefaultMethod
      drivenName=Pose_7
      functionType=DefaultFunction
      targetScale=1
      boneTransforms [{targetRotation=<rot -7.4726 -82.8401 64.7082>, targetTranslation=<15.5111 0 0>}]
      customCurve {defaultValue=3.4028234663852886e+38, postInfinityExtrap=RCCE_Constant, preInfinityExtrap=RCCE_Constant, keys []}
      targetRotation=<rot 0 0 0>
     }
     {
      bApplyCustomCurve=false
      bIsHidden=false
      distanceMethod=DefaultMethod
      drivenName=Pose_8
      functionType=DefaultFunction
      targetScale=1
      boneTransforms [{targetRotation=<rot 0.1425 2.4363 -82.9397>, targetTranslation=<15.5111 0 0>}]
      customCurve {defaultValue=3.4028234663852886e+38, postInfinityExtrap=RCCE_Constant, preInfinityExtrap=RCCE_Constant, keys []}
      targetRotation=<rot 0 0 0>
     }
    ]
    rBFParams {distanceMethod=Quaternion, function=Linear, radius=100}
    sourceBones [{boneName=upperarm_l}]
   }
   n5 = Pose Driver - Source: upperarm_r!(SourcePose:Pose Link Structure=@n4) -> (Pose:Pose Link Structure) {
    is=title AnimGraphNode_PoseDriver
    driveSource=Translation
    poseAsset=&M_MED_Commando_Guan_Yu_PoseDriver_PoseAsset
    onlyDriveBones [{boneName=pd_shoulderpad_r}]
    poseTargets [
     {
      bApplyCustomCurve=false
      bIsHidden=false
      distanceMethod=DefaultMethod
      drivenName=Pose_0
      functionType=DefaultFunction
      targetScale=1
      boneTransforms [{targetRotation=<rot -64.7854 -3.089 6.0959>, targetTranslation=<-15.5111 0 0>}]
      customCurve {defaultValue=3.4028234663852886e+38, postInfinityExtrap=RCCE_Constant, preInfinityExtrap=RCCE_Constant, keys []}
      targetRotation=<rot 0 0 0>
     }
     {
      bApplyCustomCurve=false
      bIsHidden=false
      distanceMethod=DefaultMethod
      drivenName=Pose_3
      functionType=DefaultFunction
      targetScale=1
      boneTransforms [{targetRotation=<rot -80.0317 -12.5016 15.1498>, targetTranslation=<-15.5111 0 0>}]
      customCurve {defaultValue=3.4028234663852886e+38, postInfinityExtrap=RCCE_Constant, preInfinityExtrap=RCCE_Constant, keys []}
      targetRotation=<rot 0 0 0>
     }
     {
      bApplyCustomCurve=false
      bIsHidden=false
      distanceMethod=DefaultMethod
      drivenName=Pose_4
      functionType=DefaultFunction
      targetScale=1
      boneTransforms [{targetRotation=<rot -64.7854 -3.089 71.4305>, targetTranslation=<-15.5111 0 0>}]
      customCurve {defaultValue=3.4028234663852886e+38, postInfinityExtrap=RCCE_Constant, preInfinityExtrap=RCCE_Constant, keys []}
      targetRotation=<rot 0 0 0>
     }
     {
      bApplyCustomCurve=false
      bIsHidden=false
      distanceMethod=DefaultMethod
      drivenName=Pose_5
      functionType=DefaultFunction
      targetScale=1
      boneTransforms [{targetRotation=<rot -64.7854 -3.089 -89.7952>, targetTranslation=<-15.5111 0 0>}]
      customCurve {defaultValue=3.4028234663852886e+38, postInfinityExtrap=RCCE_Constant, preInfinityExtrap=RCCE_Constant, keys []}
      targetRotation=<rot 0 0 0>
     }
     {
      bApplyCustomCurve=false
      bIsHidden=false
      distanceMethod=DefaultMethod
      drivenName=Pose_6
      functionType=DefaultFunction
      targetScale=1
      boneTransforms [{targetRotation=<rot 5.3225 96.1402 -64.8218>, targetTranslation=<-15.5111 0 0>}]
      customCurve {defaultValue=3.4028234663852886e+38, postInfinityExtrap=RCCE_Constant, preInfinityExtrap=RCCE_Constant, keys []}
      targetRotation=<rot 0 0 0>
     }
     {
      bApplyCustomCurve=false
      bIsHidden=false
      distanceMethod=DefaultMethod
      drivenName=Pose_7
      functionType=DefaultFunction
      targetScale=1
      boneTransforms [{targetRotation=<rot -7.4726 -82.8401 64.7082>, targetTranslation=<-15.5111 0 0>}]
      customCurve {defaultValue=3.4028234663852886e+38, postInfinityExtrap=RCCE_Constant, preInfinityExtrap=RCCE_Constant, keys []}
      targetRotation=<rot 0 0 0>
     }
     {
      bApplyCustomCurve=false
      bIsHidden=false
      distanceMethod=DefaultMethod
      drivenName=Pose_8
      functionType=DefaultFunction
      targetScale=1
      boneTransforms [{targetRotation=<rot 0.1425 2.4363 -82.9397>, targetTranslation=<-15.5111 0 0>}]
      customCurve {defaultValue=3.4028234663852886e+38, postInfinityExtrap=RCCE_Constant, preInfinityExtrap=RCCE_Constant, keys []}
      targetRotation=<rot 0 0 0>
     }
    ]
    rBFParams {distanceMethod=Quaternion, function=Linear, radius=100}
    sourceBones [{boneName=upperarm_r}]
   }
   n6 = Pose Driver - Source: upperarm_l(SourcePose=@n5) {
    poseAsset=&M_MED_Commando_Guan_Yu_PoseDriver_PoseAsset
    onlyDriveBones [{boneName=pd_shoulderpad_l}]
    poseTargets [
     {
      bApplyCustomCurve=false
      bIsHidden=false
      distanceMethod=DefaultMethod
      drivenName=Pose_0
      functionType=DefaultFunction
      targetScale=0.164
      boneTransforms [{targetRotation=<rot -64.7854 -3.089 6.0959>, targetTranslation=<15.5111 0 0>}]
      customCurve {defaultValue=3.4028234663852886e+38, postInfinityExtrap=RCCE_Constant, preInfinityExtrap=RCCE_Constant, keys []}
      targetRotation=<rot 0 0 0>
     }
     {
      bApplyCustomCurve=false
      bIsHidden=false
      distanceMethod=DefaultMethod
      drivenName=Pose_3
      functionType=DefaultFunction
      targetScale=0.164
      boneTransforms [{targetRotation=<rot -80.0317 -12.5016 15.1498>, targetTranslation=<15.5111 0 0>}]
      customCurve {defaultValue=3.4028234663852886e+38, postInfinityExtrap=RCCE_Constant, preInfinityExtrap=RCCE_Constant, keys []}
      targetRotation=<rot 0 0 0>
     }
     {
      bApplyCustomCurve=false
      bIsHidden=false
      distanceMethod=DefaultMethod
      drivenName=Pose_4
      functionType=DefaultFunction
      targetScale=0.6929
      boneTransforms [{targetRotation=<rot -64.7854 -3.089 71.4305>, targetTranslation=<15.5111 0 0>}]
      customCurve {defaultValue=3.4028234663852886e+38, postInfinityExtrap=RCCE_Constant, preInfinityExtrap=RCCE_Constant, keys []}
      targetRotation=<rot 0 0 0>
     }
     {
      bApplyCustomCurve=false
      bIsHidden=false
      distanceMethod=DefaultMethod
      drivenName=Pose_5
      functionType=DefaultFunction
      targetScale=0.6977
      boneTransforms [{targetRotation=<rot -64.7854 -3.089 -89.7952>, targetTranslation=<15.5111 0 0>}]
      customCurve {defaultValue=3.4028234663852886e+38, postInfinityExtrap=RCCE_Constant, preInfinityExtrap=RCCE_Constant, keys []}
      targetRotation=<rot 0 0 0>
     }
     {
      bApplyCustomCurve=false
      bIsHidden=false
      distanceMethod=DefaultMethod
      drivenName=Pose_6
      functionType=DefaultFunction
      targetScale=1
      boneTransforms [{targetRotation=<rot 5.3225 96.1402 -64.8218>, targetTranslation=<15.5111 0 0>}]
      customCurve {defaultValue=3.4028234663852886e+38, postInfinityExtrap=RCCE_Constant, preInfinityExtrap=RCCE_Constant, keys []}
      targetRotation=<rot 0 0 0>
     }
     {
      bApplyCustomCurve=false
      bIsHidden=false
      distanceMethod=DefaultMethod
      drivenName=Pose_7
      functionType=DefaultFunction
      targetScale=0.8366
      boneTransforms [{targetRotation=<rot -7.4726 -82.8401 64.7082>, targetTranslation=<15.5111 0 0>}]
      customCurve {defaultValue=3.4028234663852886e+38, postInfinityExtrap=RCCE_Constant, preInfinityExtrap=RCCE_Constant, keys []}
      targetRotation=<rot 0 0 0>
     }
     {
      bApplyCustomCurve=false
      bIsHidden=false
      distanceMethod=DefaultMethod
      drivenName=Pose_8
      functionType=DefaultFunction
      targetScale=0.6977
      boneTransforms [{targetRotation=<rot 0.1425 2.4363 -82.9397>, targetTranslation=<15.5111 0 0>}]
      customCurve {defaultValue=3.4028234663852886e+38, postInfinityExtrap=RCCE_Constant, preInfinityExtrap=RCCE_Constant, keys []}
      targetRotation=<rot 0 0 0>
     }
    ]
    rBFParams {distanceMethod=Quaternion, function=Linear, radius=100}
    sourceBones [{boneName=upperarm_l}]
   }
   n7 = Pose Driver - Source: upperarm_r(SourcePose=@n6) {
    poseAsset=&M_MED_Commando_Guan_Yu_PoseDriver_PoseAsset
    onlyDriveBones [{boneName=pd_shoulderpad_r}]
    poseTargets [
     {
      bApplyCustomCurve=false
      bIsHidden=false
      distanceMethod=DefaultMethod
      drivenName=Pose_0
      functionType=DefaultFunction
      targetScale=0.164
      boneTransforms [{targetRotation=<rot -64.7854 -3.089 6.0959>, targetTranslation=<-15.5111 0 0>}]
      customCurve {defaultValue=3.4028234663852886e+38, postInfinityExtrap=RCCE_Constant, preInfinityExtrap=RCCE_Constant, keys []}
      targetRotation=<rot 0 0 0>
     }
     {
      bApplyCustomCurve=false
      bIsHidden=false
      distanceMethod=DefaultMethod
      drivenName=Pose_3
      functionType=DefaultFunction
      targetScale=0.164
      boneTransforms [{targetRotation=<rot -80.0317 -12.5016 15.1498>, targetTranslation=<-15.5111 0 0>}]
      customCurve {defaultValue=3.4028234663852886e+38, postInfinityExtrap=RCCE_Constant, preInfinityExtrap=RCCE_Constant, keys []}
      targetRotation=<rot 0 0 0>
     }
     {
      bApplyCustomCurve=false
      bIsHidden=false
      distanceMethod=DefaultMethod
      drivenName=Pose_4
      functionType=DefaultFunction
      targetScale=0.6929
      boneTransforms [{targetRotation=<rot -64.7854 -3.089 71.4305>, targetTranslation=<-15.5111 0 0>}]
      customCurve {defaultValue=3.4028234663852886e+38, postInfinityExtrap=RCCE_Constant, preInfinityExtrap=RCCE_Constant, keys []}
      targetRotation=<rot 0 0 0>
     }
     {
      bApplyCustomCurve=false
      bIsHidden=false
      distanceMethod=DefaultMethod
      drivenName=Pose_5
      functionType=DefaultFunction
      targetScale=0.6977
      boneTransforms [{targetRotation=<rot -64.7854 -3.089 -89.7952>, targetTranslation=<-15.5111 0 0>}]
      customCurve {defaultValue=3.4028234663852886e+38, postInfinityExtrap=RCCE_Constant, preInfinityExtrap=RCCE_Constant, keys []}
      targetRotation=<rot 0 0 0>
     }
     {
      bApplyCustomCurve=false
      bIsHidden=false
      distanceMethod=DefaultMethod
      drivenName=Pose_6
      functionType=DefaultFunction
      targetScale=1
      boneTransforms [{targetRotation=<rot 5.3225 96.1402 -64.8218>, targetTranslation=<-15.5111 0 0>}]
      customCurve {defaultValue=3.4028234663852886e+38, postInfinityExtrap=RCCE_Constant, preInfinityExtrap=RCCE_Constant, keys []}
      targetRotation=<rot 0 0 0>
     }
     {
      bApplyCustomCurve=false
      bIsHidden=false
      distanceMethod=DefaultMethod
      drivenName=Pose_7
      functionType=DefaultFunction
      targetScale=0.8366
      boneTransforms [{targetRotation=<rot -7.4726 -82.8401 64.7082>, targetTranslation=<-15.5111 0 0>}]
      customCurve {defaultValue=3.4028234663852886e+38, postInfinityExtrap=RCCE_Constant, preInfinityExtrap=RCCE_Constant, keys []}
      targetRotation=<rot 0 0 0>
     }
     {
      bApplyCustomCurve=false
      bIsHidden=false
      distanceMethod=DefaultMethod
      drivenName=Pose_8
      functionType=DefaultFunction
      targetScale=0.6977
      boneTransforms [{targetRotation=<rot 0.1425 2.4363 -82.9397>, targetTranslation=<-15.5111 0 0>}]
      customCurve {defaultValue=3.4028234663852886e+38, postInfinityExtrap=RCCE_Constant, preInfinityExtrap=RCCE_Constant, keys []}
      targetRotation=<rot 0 0 0>
     }
    ]
    rBFParams {distanceMethod=Quaternion, function=Linear, radius=100}
    sourceBones [{boneName=upperarm_r}]
   }
   n8 = Get bIsSkyDiving()
   n9 = Get bIsTargeting()
   Output Pose(Result=Blend Poses by bool(BlendPose_0=@n7, BlendPose_1=Component To Local(ComponentPose=@n3), bActiveValue=OR Boolean(A=@n9, B=@n8)) {blendType=Linear}) {layerGroup=None}
   n10 = "Constraint - c_waist_armor_r (P)"(ComponentPose=Local To Component(LocalPose=Copy Pose From Mesh() {bUseAttachedParent=true}), Alpha=1, ConstraintWeights_0=1) {
    alphaBoolBlend {blendOption=Linear}
    boneToModify {boneName=c_waist_armor_r}
    constraintSetup [{targetBone {boneName=thigh_twist_01_r}}]
    constraintWeights [1]
   }
   n11 = "Constraint - c_waist_armor_l (P)"(ComponentPose=@n10, Alpha=1, ConstraintWeights_0=1) {
    alphaBoolBlend {blendOption=Linear}
    boneToModify {boneName=c_waist_armor_l}
    constraintSetup [{targetBone {boneName=thigh_twist_01_l}}]
    constraintWeights [1]
   }
   n12 = Get thigh_rot_up_r() -> (thigh_rot_up_r:double)
   n13 = "Transform (Modify) Bone - Bone: dyn_chain_r_b_01"(ComponentPose=@n11, Alpha=@n12, Translation="-24.416,-21.316,-7.442", Rotation="65,45,0", Scale="1,1,1") {
    rotationMode=BMM_Additive
    rotationSpace=BCS_BoneSpace
    translationMode=BMM_Additive
    translationSpace=BCS_BoneSpace
    alphaBoolBlend {blendOption=Linear}
    boneToModify {boneName=dyn_chain_r_b_01}
    rotation {pitch=65, yaw=45}
    translation=<-24.416 -21.316 -7.442>
   }
   n14 = "Transform (Modify) Bone - Bone: dyn_chain_r_mid_01"(ComponentPose=@n13, Alpha=@n12, Translation="0,0,0", Rotation="-0.000,25,0.000", Scale="1,1,1") {
    rotationMode=BMM_Additive
    rotationSpace=BCS_BoneSpace
    translationMode=BMM_Additive
    translationSpace=BCS_BoneSpace
    alphaBoolBlend {blendOption=Linear}
    boneToModify {boneName=dyn_chain_r_mid_01}
    rotation {yaw=25}
   }
   n15 = "Transform (Modify) Bone - Bone: dyn_chain_r_b_01"(ComponentPose=@n14, Alpha=Get calf_rot_r() -> (calf_rot_r:double), Translation="1.8,0,0", Rotation="45,-45,-90", Scale="1,1,1") {
    rotationMode=BMM_Additive
    rotationSpace=BCS_BoneSpace
    translationMode=BMM_Additive
    translationSpace=BCS_BoneSpace
    alphaBoolBlend {blendOption=Linear}
    boneToModify {boneName=dyn_chain_r_b_01}
    rotation=<rot 45 -45 -90>
    translation {x=1.8}
   }
   n16 = "Transform (Modify) Bone - Bone: dyn_chain_l_back_01"(ComponentPose=@n15, Alpha=Get calf_rot_l() -> (calf_rot_l:double), Translation="0,0,0", Rotation="45,0.010,0", Scale="1,1,1") {
    rotationMode=BMM_Additive
    rotationSpace=BCS_BoneSpace
    translationMode=BMM_Additive
    translationSpace=BCS_BoneSpace
    alphaBoolBlend {blendOption=Linear}
    boneToModify {boneName=dyn_chain_l_back_01}
    rotation {pitch=45, yaw=0.01}
   }
   n17 = Get thigh_rot_up_l() -> (thigh_rot_up_l:double)
   n18 = "Transform (Modify) Bone - Bone: dyn_chain_l_back_01"(ComponentPose=@n16, Alpha=@n17, Translation="-1.193,1.048,3.470", Rotation="24.622,-41.275,-0.009", Scale="1,1,1") {
    rotationMode=BMM_Additive
    rotationSpace=BCS_BoneSpace
    translationMode=BMM_Additive
    translationSpace=BCS_BoneSpace
    alphaBoolBlend {blendOption=Linear}
    boneToModify {boneName=dyn_chain_l_back_01}
    rotation=<rot 24.622 -41.275 -0.009>
    translation=<-1.1929 1.048 3.4698>
   }
   n19 = "Transform (Modify) Bone - Bone: dyn_chain_l_mid_01"(ComponentPose=@n18, Alpha=@n17, Translation="0,0,0", Rotation="-0.026,-27.218,0.007", Scale="1,1,1") {
    rotationMode=BMM_Additive
    rotationSpace=BCS_BoneSpace
    translationMode=BMM_Additive
    translationSpace=BCS_BoneSpace
    alphaBoolBlend {blendOption=Linear}
    boneToModify {boneName=dyn_chain_l_mid_01}
    rotation=<rot -0.026 -27.2179 0.007>
   }
   n20 = RigidBody(ComponentPose=@n19) {
    bClampLinearTranslationLimitToRefPose=true
    bOverrideWorldGravity=true
    overridePhysicsAsset=&/M_MED_Commando_Guan_Yu_Physics
    simulationSpace=BaseBoneSpace
    alphaBoolBlend {blendOption=Linear}
    baseBoneRef {boneName=pelvis}
    componentAppliedLinearAccClamp=<3000 3000 3000>
    componentLinearAccScale=<0.3 0.3 0.3>
    componentLinearVelScale=<0.3 0.3 0.3>
    overrideWorldGravity {z=-980}
   }
   n21 = Anim Dynamics - Bone: dyn_noseRing!(ComponentPose=@n20, Alpha=1, GravityScale:float=1, GravityOverride:Vector="0,0,-980", ExternalForce="0,0,0") -> (Pose) {
    angularDampingOverride=0.9999
    angularSpringConstant=1000
    bGravityOverrideInSimSpace=true
    bOverrideAngularDamping=true
    simulationSpace=RootRelative
    alphaBoolBlend {blendOption=Linear}
    boundBone {boneName=dyn_noseRing}
    componentAppliedLinearAccClamp=<5000 5000 5000>
    componentLinearAccScale=<0.6 0.6 0.6>
    componentLinearVelScale=<0.6 0.6 0.6>
    gravityOverride {z=-980}
    physicsBodyDefinitions [{
      sphereCollisionRadius=0
      boundBone {boneName=dyn_noseRing}
      boxExtents=<10 20 20>
      constraintSetup {angularLimitsMax=<0 90 0>, angularLimitsMin=<0 -75 0>, angularTarget=<0 0 0>}
      localJointOffset=<3 0 -2>
     }]
    relativeSpaceBone {boneName=head}
   }
   n22 = Anim Dynamics - Bone: dyn_sleeve_b(ComponentPose=@n21, Alpha=0.125, ExternalForce="0,0,0") {
    angularSpringConstant=325
    bAngularSpring=true
    bGravityOverrideInSimSpace=true
    bLinearSpring=true
    lODThreshold=1
    linearSpringConstant=325
    simulationSpace=Actor
    alphaBoolBlend {blendOption=Linear}
    boundBone {boneName=dyn_sleeve_b}
    physicsBodyDefinitions [{
      sphereCollisionRadius=0
      boundBone {boneName=dyn_sleeve_b}
      boxExtents=<30 30 30>
      constraintSetup {
       angularTargetAxis=AxisY
       angularLimitsMax=<0 15 15>
       angularLimitsMin=<0 -15 -15>
       angularTarget=<0 0 0>
       linearAxesMax=<10 10 10>
       linearAxesMin=<0 -10 -10>
      }
      localJointOffset=<0 0 0>
     }]
   }
   n23 = Anim Dynamics - Bone: dyn_sleeve_f(ComponentPose=@n22, Alpha=0.125, ExternalForce="0,0,0") {
    angularSpringConstant=325
    bAngularSpring=true
    bGravityOverrideInSimSpace=true
    bLinearSpring=true
    lODThreshold=1
    linearSpringConstant=325
    simulationSpace=Actor
    alphaBoolBlend {blendOption=Linear}
    boundBone {boneName=dyn_sleeve_f}
    physicsBodyDefinitions [{
      sphereCollisionRadius=0
      boundBone {boneName=dyn_sleeve_f}
      boxExtents=<30 30 30>
      constraintSetup {
       angularTargetAxis=AxisY
       angularLimitsMax=<0 15 15>
       angularLimitsMin=<0 -15 -15>
       angularTarget=<0 0 0>
       linearAxesMax=<10 10 10>
       linearAxesMin=<0 -10 -10>
      }
      localJointOffset=<0 0 0>
     }]
   }
   n24 = Anim Dynamics - Bone: dyn_sleeve_r(ComponentPose=@n23, Alpha=0.125, ExternalForce="0,0,0") {
    angularSpringConstant=325
    bAngularSpring=true
    bGravityOverrideInSimSpace=true
    bLinearSpring=true
    lODThreshold=1
    linearSpringConstant=325
    simulationSpace=Actor
    alphaBoolBlend {blendOption=Linear}
    boundBone {boneName=dyn_sleeve_r}
    physicsBodyDefinitions [{
      sphereCollisionRadius=0
      boundBone {boneName=dyn_sleeve_r}
      boxExtents=<30 30 30>
      constraintSetup {
       angularTargetAxis=AxisY
       angularLimitsMax=<0 15 15>
       angularLimitsMin=<0 -15 -15>
       angularTarget=<0 0 0>
       linearAxesMax=<10 10 10>
       linearAxesMin=<0 -10 -10>
      }
      localJointOffset=<0 0 0>
     }]
   }
   n25 = Anim Dynamics - Bone: dyn_sleeve_l(ComponentPose=@n24, Alpha=0.125, ExternalForce="0,0,0") {
    angularSpringConstant=325
    bAngularSpring=true
    bGravityOverrideInSimSpace=true
    bLinearSpring=true
    lODThreshold=1
    linearSpringConstant=325
    simulationSpace=Actor
    alphaBoolBlend {blendOption=Linear}
    boundBone {boneName=dyn_sleeve_l}
    physicsBodyDefinitions [{
      sphereCollisionRadius=0
      boundBone {boneName=dyn_sleeve_l}
      boxExtents=<30 30 30>
      constraintSetup {
       angularTargetAxis=AxisY
       angularLimitsMax=<0 15 15>
       angularLimitsMin=<0 -15 -15>
       angularTarget=<0 0 0>
       linearAxesMax=<10 10 10>
       linearAxesMin=<0 -10 -10>
      }
      localJointOffset=<0 0 0>
     }]
   }
   Save cached pose 'prePDCache'(Pose=Component To Local(ComponentPose=@n25)) {cacheName=prePDCache}
  ]}
 EventGraph {nodes [
   Event BlueprintUpdateAnimation()
   n1 = GetOwnerSkeletalMeshComponent()
   n2 = GetDeltaTransformFromRefPose(self=@n1, BoneName=thigh_r, BaseName=pelvis)
   n3 = MapRangeClamped(Value=BreakRotator(InRot=BreakTransform(InTransform=@n2).Rotation).Yaw, InRangeA=-15, InRangeB=-100)
   > Set thigh_rot_up_r(thigh_rot_up_r:double=@n3) -> (Output_Get:double)
   n4 = BreakRotator(InRot=BreakTransform(InTransform=GetDeltaTransformFromRefPose(self=@n1, BoneName=thigh_r, BaseName=calf_r)).Rotation)
   > Set calf_rot_r(calf_rot_r:double=float * float(A=MapRangeClamped(Value=@n3, InRangeA=0, InRangeB=1, OutRangeA=1, OutRangeB=0), B=MapRangeClamped(Value=@n4.Yaw, InRangeA=0, InRangeB=-120)) {is=title K2Node_PromotableOperator}) -> (Output_Get:double)
   n5 = GetOwnerSkeletalMeshComponent()
   n6 = BreakTransform(InTransform=GetDeltaTransformFromRefPose(self=@n5, BoneName=thigh_l, BaseName=calf_l))
   > Set calf_rot_l(calf_rot_l:double=MapRangeClamped(Value=BreakRotator(InRot=@n6.Rotation).Yaw, InRangeA=0, InRangeB=-110)) -> (Output_Get:double)
   n7 = BreakTransform(InTransform=GetDeltaTransformFromRefPose(self=@n5, BoneName=upperarm_l, BaseName=spine_05))
   > Set arm_rot_up_l(arm_rot_up_l:double=MapRangeClamped(Value=BreakRotator(InRot=@n7.Rotation).Yaw, InRangeA=0, InRangeB=120)) -> (Output_Get:double)
   n8 = BreakTransform(InTransform=GetDeltaTransformFromRefPose(self=GetOwnerSkeletalMeshComponent(), BoneName=thigh_l, BaseName=pelvis))
   > Set thigh_rot_up_l(thigh_rot_up_l:double=MapRangeClamped(Value=BreakRotator(InRot=@n8.Rotation).Yaw, InRangeA=0, InRangeB=-110)) -> (Output_Get:double)
  ]}
]
variables [
 arm_rot_up_l {type=number}
 calf_rot_l {type=number}
 calf_rot_r {type=number}
 thigh_rot_up_l {type=number}
 thigh_rot_up_r {type=number}
]
