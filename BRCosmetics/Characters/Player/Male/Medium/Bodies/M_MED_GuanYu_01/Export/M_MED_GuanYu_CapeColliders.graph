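# Physics asset of cape colliders for the male medium GuanYu player body, apparently for cape simulation to collide against.
# Capsules cover the pelvis, spine, clavicles and neck, with a sphere on the head and multiple overlapping capsules on spine_02 and spine_05.
# Five capsules also ride on third-link bones of left and right dyn_chain bones, adding colliders that follow those chains.
# Arm, leg and foot bodies have no shapes, and no constraints exist.
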PhysicsAsset /BRCosmetics/Characters/Player/Male/Medium/Bodies/M_MED_GuanYu_01/Export/M_MED_GuanYu_CapeColliders
bodies {
 calf_l []
 calf_r []
 clavicle_l [{shapeType=Capsule, length=6.2636, radius=18.8927, center=<14.5205 -0.5706 5.8017>, rotation=<rot 75.9352 -0.0001 175.5622>}]
 clavicle_r [{shapeType=Capsule, length=3.7636, radius=15.3927, center=<-14.0739 0.5385 -2.3187>, rotation=<rot 55.9606 2.7096 -1.9253>}]
 dyn_chain_l_back_03 [{
   shapeType=Capsule
   length=20.75
   radius=15
   shapeName=dyn_chain_l_back_03_capsule
   center=<-6.0282 -1.4187 -4.8095>
   rotation=<rot -86.5442 -89.854 90.1519>
  }]
 dyn_chain_l_mid_03 [{
   shapeType=Capsule
   length=21.25
   radius=13
   shapeName=dyn_chain_l_mid_03_capsule
   center=<-8.1616 0.0215 1.4166>
   rotation=<rot -80 -170.0001 170.0004>
  }]
 dyn_chain_r_b_03 [{
   shapeType=Capsule
   length=24.75
   radius=15.5
   shapeName=dyn_chain_r_b_03_capsule
   center=<3.6309 -5.3801 -1.4055>
   rotation=<rot -88.2657 104.9298 -95.081>
  }]
 dyn_chain_r_f_03 [{shapeType=Capsule, length=15, radius=15, shapeName=dyn_chain_r_f_03_capsule, rotation=<rot -80.0001 -169.9998 179.9998>}]
 dyn_chain_r_mid_03 [{
   shapeType=Capsule
   length=26.25
   radius=13.5
   shapeName=dyn_chain_r_mid_03_capsule
   center=<-0.0185 7.1351 -0.2153>
   rotation=<rot -80 -169.9999 169.9997>
  }]
 foot_l []
 foot_r []
 head [{shapeType=Sphere, radius=17.3125, center=<11.028 -0.174 0>}]
 lowerarm_l []
 lowerarm_r []
 neck_02 [{shapeType=Capsule, length=13.4375, radius=7.3175, center=<-0.2642 -0.3408 0>, rotation=<rot 90 -171.1193 -175.3978>}]
 pelvis [{shapeType=Capsule, length=7.9062, radius=21.0388, center=<-2.6215 -3.9147 0>}]
 spine_01 [{shapeType=Capsule, length=4.7187, radius=18.4138, center=<-2.6215 -1.6627 0>}]
 spine_02 [
  {shapeType=Capsule, length=5.8437, radius=15.695, center=<-2.6215 -3.2548 0.0001>}
  {shapeType=Capsule, length=29.875, radius=11.1013, center=<2.5024 0.4992 3.9825>, rotation=<rot 90 0.0049 0.0049>}
  {shapeType=Capsule, length=29.875, radius=11.1013, center=<2.4704 0.4991 -3.9825>, rotation=<rot 90 -0.0396 -0.0396>}
 ]
 spine_03 [{shapeType=Capsule, length=5.1562, radius=14.7575, center=<-1.953 -1.3275 0.0001>}]
 spine_04 [{shapeType=Capsule, length=12.1562, radius=16.1325, center=<2.1467 -1.2748 0.0002>}]
 spine_05 [
  {shapeType=Capsule, length=16.0937, radius=18.945, center=<3.6172 -0.7186 0>}
  {shapeType=Capsule, length=12.7812, radius=17.0625, center=<0.7875 3.1143 0>}
 ]
 thigh_l []
 thigh_r []
 upperarm_l []
 upperarm_r []
}
constraints []
